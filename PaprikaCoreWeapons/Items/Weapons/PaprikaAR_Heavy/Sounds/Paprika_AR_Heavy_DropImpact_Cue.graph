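# Sound cue for the drop impact of the Paprika heavy AR weapon item, in the loot collect sound class.
# A random node picks one of three equally weighted impact waves and feeds it through a modulator for variation.
# The cue overrides attenuation for a short audible range and overrides concurrency to cap simultaneous instances, preventing new plays once the limit is reached.
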SoundCue /PaprikaCoreWeapons/Items/Weapons/PaprikaAR_Heavy/Sounds/Paprika_AR_Heavy_DropImpact_Cue
root=Modulator_0
nodes [
 {id=Modulator_0, children [Random_1]}
 {id=Random_1, children [WavePlayer_0,WavePlayer_1,WavePlayer_2], properties {weights [1,1,1]}}
 {id=WavePlayer_0, properties {soundWaveAssetPtr=&/Paprika_AR_Heavy_DropImpact_03}}
 {id=WavePlayer_1, properties {soundWaveAssetPtr=&/Paprika_AR_Heavy_DropImpact_02}}
 {id=WavePlayer_2, properties {soundWaveAssetPtr=&/Paprika_AR_Heavy_DropImpact_01}}
]
properties {
 bOverrideAttenuation=true
 bOverrideConcurrency=true
 maxDistance=1000
 soundClassObject=&Fort_Loot_Collect
 volumeMultiplier=0.8
 attenuationOverrides {falloffDistance=1000, attenuationShapeExtents {x=0}}
 concurrencyOverrides {maxCount=3, resolutionRule=PreventNew}
}
