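# Sound cue for picking up the Paprika heavy assault rifle, routed to the loot-collect sound class.
# A mixer layers the main pickup wave, run through a modulator, with a slightly quieter sweetener picked at random from two equally weighted variants.
# The cue plays slightly pitched down, with shared weapon-pickup attenuation and a pickup concurrency set limiting simultaneous plays.
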SoundCue /PaprikaCoreWeapons/Items/Weapons/PaprikaAR_Heavy/Sounds/Paprika_AR_Heavy_Pickup_Cue
root=Mixer_0
nodes [
 {id=Mixer_0, children [Modulator_0,Random_0], properties {inputVolume [1,0.7]}}
 {id=Modulator_0, children [WavePlayer_0]}
 {id=Random_0, children [WavePlayer_1,WavePlayer_2], properties {weights [1,1]}}
 {id=WavePlayer_0, properties {soundWaveAssetPtr=&/Paprika_AR_Heavy_PickUp}}
 {id=WavePlayer_1, properties {soundWaveAssetPtr=&/Paprika_AR_Heavy_Pickup_Sweetner_02}}
 {id=WavePlayer_2, properties {soundWaveAssetPtr=&/Paprika_AR_Heavy_Pickup_Sweetner_01}}
]
properties {
 attenuationSettings=&Fort_Weapon_Pickup_Att
 maxDistance=3500
 pitchMultiplier=0.9
 soundClassObject=&Fort_Loot_Collect
 concurrencyOverrides {maxCount=1, resolutionRule=PreventNew}
 concurrencySet [&PickupMaximum_Oldest]
}
