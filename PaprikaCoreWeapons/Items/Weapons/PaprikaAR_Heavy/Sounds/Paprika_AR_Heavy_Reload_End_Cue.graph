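# Sound cue for the end of the Paprika heavy assault rifle reload, played as weapon foley.
# A Random node picks with equal weight between two of the rifle's Equip waves, reusing them for the reload, and feeds a Modulator at the root.
# It uses the shared weapon reload attenuation and foley sound class, with reduced volume and a concurrency override.
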SoundCue /PaprikaCoreWeapons/Items/Weapons/PaprikaAR_Heavy/Sounds/Paprika_AR_Heavy_Reload_End_Cue
root=Modulator_0
nodes [
 {id=Modulator_0, children [Random_1]}
 {id=Random_1, children [WavePlayer_0,WavePlayer_1], properties {weights [1,1]}}
 {id=WavePlayer_0, properties {soundWaveAssetPtr=&/Paprika_AR_Heavy_Equip_01}}
 {id=WavePlayer_1, properties {soundWaveAssetPtr=&/Paprika_AR_Heavy_Equip_02}}
]
properties {
 attenuationSettings=&Fort_Weapon_Reload_Att
 bDebug=true
 bOverrideConcurrency=true
 maxDistance=2250
 soundClassObject=&Fort_WeaponFoley
 volumeMultiplier=0.301
}
